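# Sound cue for limited-time mode score music, named for the start of a final-10 countdown.
# A quality-level node at the root sends the two lower slots to an orchestral medium-threat combat ending wave and the two higher slots to the dedicated count-final score wave.
# It plays at reduced volume through an important in-game mission sound class, with a concurrency override capping it at two instances.
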SoundCue /Game/Athena/Sounds/LTMs/Score/CountdownLTM_Final10Start_Cue
root=QualityLevel_0
nodes [
 {id=QualityLevel_0, children [WavePlayer_1,WavePlayer_1,WavePlayer_0,WavePlayer_0]}
 {id=WavePlayer_0, properties {soundWaveAssetPtr=&/ScoreLTM_InGamePop_CountFinal_06}}
 {id=WavePlayer_1, properties {soundWaveAssetPtr=/Game/Sounds/Fort_Music/Orchestral_Score/Orch_Combat/Combat_Med_Threat_End_01}}
]
properties {
 bOverrideConcurrency=true
 maxDistance=2097152
 soundClassObject=&Fort_Mission_Ingame_Important_PSM
 volumeMultiplier=0.25
 concurrencyOverrides {maxCount=2}
}
