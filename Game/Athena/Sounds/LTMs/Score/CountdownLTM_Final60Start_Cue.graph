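# Sound cue for a countdown LTM, named for the start of the final 60 seconds, playing the Cobalt 60-second warning sound effect.
# A wave player feeds an Enveloper that lowers pitch and fades the volume out just before one second.
# A Mixer halves that signal and has an empty second input.
# The cue uses the in-game mission sound class, a low overall volume and a concurrency override limiting simultaneous plays to two.
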SoundCue /Game/Athena/Sounds/LTMs/Score/CountdownLTM_Final60Start_Cue
root=Mixer_0
nodes [
 {
  id=Enveloper_0
  children [WavePlayer_1]
  properties {pitchCurve {editorCurveData {keys [<at 0 0.7>]}}, volumeCurve {editorCurveData {keys [<at 0 1>,<at 0.7636 1>,<at 0.9797 0>]}}}
 }
 {id=Mixer_0, children [Enveloper_0,null], properties {inputVolume [0.5,1]}}
 {id=WavePlayer_1, properties {soundWaveAssetPtr=/Game/Sounds/LTM/Cobalt/Cobalt_60SecondsSFX_Warning}}
]
properties {
 bOverrideConcurrency=true
 maxDistance=2097152
 soundClassObject=/Game/Sounds/FortSoundClassesAndModes/Fort_Mission_IngameSounds
 volumeMultiplier=0.2
 concurrencyOverrides {maxCount=2}
}
