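# Sound cue for the final countdown tick, filed under limited-time mode score sounds.
# A single wave player feeds the root with the normal tick sample, reused at slightly raised pitch and reduced volume.
# It uses an important in-game mission sound class, overrides concurrency to cap simultaneous plays at two, and sets a very large max distance.
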SoundCue /Game/Athena/Sounds/LTMs/Score/CountdownLTM_FinalTick_Cue
root=WavePlayer_1
nodes [{id=WavePlayer_1, properties {soundWaveAssetPtr=&/CountdownLTM_NormalTick_01}}]
properties {
 bOverrideConcurrency=true
 maxDistance=2097152
 pitchMultiplier=1.05
 soundClassObject=&Fort_Mission_Ingame_Important_PSM
 volumeMultiplier=0.35
 concurrencyOverrides {maxCount=2}
}
