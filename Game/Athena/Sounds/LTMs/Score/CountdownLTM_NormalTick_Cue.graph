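# Sound cue for the normal countdown tick in Athena limited-time-mode score audio.
# A single wave player plays the Infiltration UI time tick wave at reduced volume and slightly raised pitch.
# It overrides concurrency to cap simultaneous instances and references Fort_Enemy_Damaged_PSM as its sound class.
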SoundCue /Game/Athena/Sounds/LTMs/Score/CountdownLTM_NormalTick_Cue
root=WavePlayer_1
nodes [{id=WavePlayer_1, properties {soundWaveAssetPtr=/Game/Sounds/LTM/Infiltration/Infiltration_UI_Time_Tick}}]
properties {
 bOverrideConcurrency=true
 maxDistance=2097152
 pitchMultiplier=1.05
 soundClassObject=/Game/Sounds/FortSoundClassesAndModes/Fort_Enemy_Damaged_PSM
 volumeMultiplier=0.25
 concurrencyOverrides {maxCount=2}
}
