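# Looping countdown tick Sound Cue for limited-time-mode scoring, built by reusing the safe zone holding loop pitched down and boosted in volume.
# A Quality Level node feeds the standard wave to the first two tiers and the "_opt" variant to the last two.
# A single-input Mixer adds gain on top of the cue's volume multiplier.
# It overrides concurrency to two instances and uses the Fort_Enemy_Damaged_PSM sound class.
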SoundCue /Game/Athena/Sounds/LTMs/Score/CountdownLTM_TickLoop_Cue
root=Mixer_0
nodes [
 {id=Mixer_0, children [QualityLevel_0], properties {inputVolume [2]}}
 {id=QualityLevel_0, children [WavePlayer_5,WavePlayer_5,WavePlayer_0,WavePlayer_0]}
 {id=WavePlayer_0, properties {bLooping=true, soundWaveAssetPtr=/Game/Athena/Sounds/Shield/SFX_SafeZone_HoldingLoop_opt}}
 {id=WavePlayer_5, properties {bLooping=true, soundWaveAssetPtr=/Game/Athena/Sounds/Shield/SFX_SafeZone_HoldingLoop}}
]
properties {
 bOverrideConcurrency=true
 maxDistance=2097152
 pitchMultiplier=0.7
 soundClassObject=/Game/Sounds/FortSoundClassesAndModes/Fort_Enemy_Damaged_PSM
 volumeMultiplier=2.5
 concurrencyOverrides {maxCount=2}
}
