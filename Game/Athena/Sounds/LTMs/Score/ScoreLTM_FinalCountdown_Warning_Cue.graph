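# Sound cue for the final countdown warning pop in score-based limited-time modes, routed through an important in-game mission sound class.
# A quality-level node serves the standard wave to the first two quality tiers and an optimized variant to the other two.
# A concurrency override allows only one instance at a time, blocking new plays while it sounds.
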SoundCue /Game/Athena/Sounds/LTMs/Score/ScoreLTM_FinalCountdown_Warning_Cue
root=QualityLevel_0
nodes [
 {id=QualityLevel_0, children [WavePlayer_1,WavePlayer_1,WavePlayer_2,WavePlayer_2]}
 {id=WavePlayer_1, properties {soundWaveAssetPtr=&/ScoreLTM_InGamePop_FinalCountdown_Warning_07}}
 {id=WavePlayer_2, properties {soundWaveAssetPtr=&/ScoreLTM_InGamePop_FinalCountdown_Warning_07_opt}}
]
properties {
 bOverrideConcurrency=true
 maxDistance=2097152
 soundClassObject=&Fort_Mission_Ingame_Important_PSM
 volumeMultiplier=1
 concurrencyOverrides {maxCount=1, resolutionRule=PreventNew}
}
